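; Rift Gauntlets dash ability, extending a parent ability with forward punch, side dash and slam variants.
; Movement input and targeting pick a guided-motion movement mode, which passes dash type back as an optional float through gameplay events.
; Start/end events, landing checks and hotfix-tuned delays drive montage sections, whose notifies trigger targeted, mostly async damage and cues.
; Ending cleans up tasks, effects and camera, then destroys the weapon when out of ammo.
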(event Ability|EventActivateAbility
  (bind _hovermovementmode (Variables|MovementMode|GetHoverMovementMode))
  ;[0.05009 0.15 9.2e-05 0.5] ActivateAbility: Majority of the Dash functionality is run in the parent. This class primarily takes care of Rift Gauntlet specific events.
  (bind _returnvalue (Ability|CommitAbility))
  (bind _output_get (Variables|Transient|SetbCommitSuccessful _returnvalue))
  (if _output_get
    (General|CacheVariables) ; Cache variables before parent activation\n starts running meaningful logic.
    (|CallParentFunction)
    (if (Variables|Animation|GetbManipulateMontageSectionPlayRate)
      (Animation|SetBindtoMontageSectionChanged true) ; We basically want to know when the initial section is \nplayed out so we can adjust the playrate of the \nmontage for balance reasons.
      (else
        _))
    (Ability|ResetVariables)
    (MovementMode|ExitMovementStates)
    (bind _returnvalue_2 (Ability|ApplyGameplayEffectToOwner (Variables|GameplayEffects|GetActiveTagGameplayEffectClass)))
    (Variables|Transient|SetActiveTagGameplayEffectHandle _returnvalue_2)
    (bind _returnvalue_3 (Ability|ApplyGameplayEffectToOwner (Variables|GameplayEffects|GetBlockChangeEquipmentGameplayEffect)))
    (Variables|Transient|SetBlockChangeEquipmentEffectHandle _returnvalue_3)
    (MovementMode|SetEnableHighAccelerationTolerance true)
    (Utilities|FlowControl|Sequence
      (:then_0
        ;;[0.04245 0.06021 0.15 0.5] Applying minor velocity on the Z axis to hover up.
        (Ability|HasAuthority
          (:False
            (if (Utilities|IsValidClass _hovermovementmode)
              (bind _avatarplayerpawn (Variables|Transient|GetAvatarPlayerPawn)
                (:then
                  (bind _returnvalue_4 (Game|ApplyGuidedMotionMME _avatarplayerpawn _hovermovementmode))
                  (Variables|Transient|SetHoverMovementModeRuntime _returnvalue_4))
                (:else))))))
      (:then_1
        ;;[0.15 0.05466 0.07488 0.5] Track when the dash MME starts and ends on the client and server.
        (bind _asynctaskproxy (Ability|Tasks|WaitGameplayEvent (Variables|MovementMode|GetDashStartEventTag) 0 true)
          (:then
            (Variables|Transient|Async|SetAsyncTask_WaitDashMovementModeStartEvent _asynctaskproxy)
            (bind _asynctaskproxy_2 (Ability|Tasks|WaitGameplayEvent (Variables|MovementMode|GetDashEndEventTag) 0 true)
              (:then
                (Variables|Transient|Async|SetAsyncTask_WaitDashMovementModeEndEvent _asynctaskproxy_2)
                (Camera|SetEnableDashCameraModifier true)
                (Utilities|FlowControl|Sequence
                  (:then_0
                    (Utilities|FlowControl|DelayUntilNextFrame) ; Delayed to help perf.
                    (Ability|ExecuteGameplayCueOnOwner (Utilities|Struct|BreakGameplayCueTag (Variables|GameplayCueTag|GetDashActivateGameplayCueTag))))
                  (:then_1
                    ;;[0.0075 0.15 0.1238 0.5] Listen to side dash behavior execution.
                    (bind (_asynctaskproxy_3 _payload) (Ability|Tasks|WaitGameplayEvent "(TagName=\"RiftGauntlets.Event.Ability.Dash.MME.OptionalFloat.0\")" 0 true)
                      (:then
                        (Variables|Transient|Async|SetAsyncTask_WaitOptionalBool _asynctaskproxy_3)
                        ;;[0.15 0.03521 0.1172 0.5] Ability timeout incase of MME failing outside the ability system.
                        (bind _asynctaskproxy_4 (Ability|Tasks|WaitDelay (ScalableFloat|GetValueatLevel (Variables|Hotfix|GetHotfix_AbilityTimeoutDelay)))
                          (:then
                            (Variables|Transient|Async|SetAsyncTask_TimeoutDelay _asynctaskproxy_4)
                            (Character|UnCrouch (Variables|Transient|GetAvatarPlayerPawn)))
                          (:OnFinish
                            (Ability|CancelAbility))))
                      (:EventReceived
                        (bind _output_get_2 (Variables|Transient|SetCurrentDashType (General|OptionalFloattoDashType (.EventMagnitude (Utilities|Struct|BreakGameplayEventData _payload)))))
                        (if (Utilities|Enum|Equal(Enum) _output_get_2 "NewEnumerator1")
                          (Ability|Animation|MontageJumptoSection "SideDash")
                          (Ability|RemoveGameplayEffectFromOwnerWithAssetTags "(GameplayTags=((TagName=\"RiftGauntlets.Stance.Dash\")))") ; Remove all the stance GEs to reset them.
                        ))))))
              (:EventReceived
                (CallFunction|HandleDashMovementModeExtEnded))))
          (:EventReceived
            (CallFunction|HandleDashMovementModeExtStarted)))))
    (else
      (Ability|EndAbility))))

(event Custom|HandleDashMovementModeExtStarted
  ;[0.01043 0.08902 0.1042 0.5] HandleDashMovementModeExtStarted: Dash has started, start tracing ahead for actors in the way.
  (CallFunction|GiveTillLandFallImmunity)
  (Variables|Transient|SetDashStartGameTime (Utilities|Time|GetGameTimeinSeconds))
  (Utilities|FlowControl|Sequence
    (:then_0
      (bind _asynctaskproxy (GameplayTasks|TaskWaitDelay 0 (ScalableFloat|GetValueatLevel (Variables|Hotfix|GetHotfix_DelayBeforeJumpToForwardPunchSection)))
        (:then
          (Variables|Transient|Async|SetAsyncTask_DelayBeforeJumpToForwardPunchSection _asynctaskproxy)
          ;;[0.1458 0.03261 0.15 0.5] Setting up behavior and listen events for when the dash is going to be a slam.
          (switch Utilities|FlowControl|Switch|SwitchonRiftGauntletsDashTypeEnum (Variables|Transient|GetCurrentDashType)
            (:NewEnumerator0
              ;;;[0.009487 0.15 0.1108 0.5] Contextual tutorial for side dash.
              (Ability|HasAuthority
                (:False
                  (if (ScalableFloat|GetValueasBool (Variables|Hotfix|GetHotfix_EnableContextualDashTutorial))
                    (Ability|SendGameplayEventtoActor (Variables|Transient|GetAvatarPlayerPawn) "(TagName=\"RiftGauntlets.Tutorial.Ability.Dash.Directional.Start\")")))))
            (:NewEnumerator1
              (Ability|HasAuthority
                (:False
                  (if (ScalableFloat|GetValueasBool (Variables|Hotfix|GetHotfix_EnableContextualDashTutorial))
                    (Ability|SendGameplayEventtoActor (Variables|Transient|GetAvatarPlayerPawn) "(TagName=\"RiftGauntlets.Tutorial.Ability.Dash.Directional.End\")")))))
            (:NewEnumerator2
              (CallFunction|SetPawnCollision (Variables|Transient|GetAvatarPlayerPawn))
              (EventDispatchers|BindEventtoOnPawnLanded (Variables|Transient|GetAvatarPlayerPawn) (EventDispatchers|CreateEvent :!event Custom|HandlePawnLandedFromSlam))
              (bind _asynctaskproxy_2 (Ability|Tasks|WaitGameplayTagAdd "(TagName=\"Movement.Status.SlopeSliding\")" 0 true)
                (:then
                  (Utilities|Array|Add (Variables|Transient|Async|GetAsyncTask_WaitForSlamLand) _asynctaskproxy_2)
                  (bind _asynctaskproxy_3 (Ability|Tasks|WaitGameplayTagAdd "(TagName=\"Gameplay.Status.InWaterBody\")" 0 true)
                    (:then
                      (Utilities|Array|Add (Variables|Transient|Async|GetAsyncTask_WaitForSlamLand) _asynctaskproxy_3)
                      (bind _asynctaskproxy_4 (Ability|Tasks|WaitGameplayTagAdd "(TagName=\"Granted.Creative.CreativeEffect.RecentlyUsed.Crashpad\")" 0 true)
                        (:then
                          (Utilities|Array|Add (Variables|Transient|Async|GetAsyncTask_WaitForSlamLand) _asynctaskproxy_4)
                          (bind _returnvalue (Ability|ApplyGameplayEffectToOwner "/RiftGauntletsGameplay/Gameplay/Dash/GE_RiftGauntlets_Dash_Slam_PreventMoveAndJump.GE_RiftGauntlets_Dash_Slam_PreventMoveAndJump_C")) ; Adding the prevent move right as the slam dash begins \nso that it is well into being active when the slam ends.
                          (Variables|Transient|SetSlamDashPreventMove _returnvalue))
                        (:Added
                          (Ability|EndAbility))))
                    (:Added
                      (Default|WaterDepthCheck
                        (:End
                          (Ability|EndAbility))
                        (:Don't%20End
                          (CallFunction|HandlePawnLandedfromSlam))))))
                (:Added
                  (CallFunction|HandlePawnLandedfromSlam))))))
        (:OnFinish
          (if (Utilities|Enum|Equal(Enum) (Variables|Transient|GetCurrentDashType) "NewEnumerator0")
            (Ability|JumptoMontageForwardPunchSection)))) ; When forward dashing we want to jump to the punch montage section \nbefore the MME actually stops. This is what this delay is meant to help with.
    )
    (:then_1
      (if (and (Variables|General|GetbPerformContinuousDashTargetingBehavior) (not (Utilities|Enum|Equal(Enum) (Variables|Transient|GetCurrentDashType) "NewEnumerator2")))
        (bind (_asynctaskproxy_5 _targetdata) (Ability|Tasks|DoTargetSelectionContinuous "ContinuousTargetingTask" "(TagName=\"Targeting.ApplicationTag.0\")" false)
          (:then
            (bind _output_get (Variables|Transient|Async|SetAsyncTask_ContinuousTargeting _asynctaskproxy_5)))
          (:Targeted
            (bind _bstopdash (General|ProcessContinuousTargetingTaskTargetData _targetdata))
            (if _bstopdash
              (GameplayTasks|EndTask _output_get)
              (Utilities|FlowControl|Sequence
                (:then_0
                  (MovementMode|StopDashMovementMode)
                  (CallFunction|EndDashAbilityBehavior))
                (:then_1
                  ;;[0.07592 0.06551 0.15 0.5] Smoother finish to the dash when not colliding with a moving pawn in the way.
                  (bind _returnvalue_2 (ScalableFloat|GetValueatLevel (Variables|Hotfix|GetHotfix_IgnoreCollisionWithPawnsAfterContinuousTargetDetect)))
                  (if (> _returnvalue_2 0)
                    (for _array_element (Ability|TargetData|GetAllActorsfromTargetData _targetdata)
                      (bind _asfort_player_pawn (Utilities|Casting|CastToFortPlayerPawn _array_element))
                      (bind _avatarplayerpawn (Variables|Transient|GetAvatarPlayerPawn)
                        (:then
                          (Collision|AddIgnoredActor _avatarplayerpawn _asfort_player_pawn _returnvalue_2))
                        (:else)))))))))))))

(event Custom|ApplyDamageToDashCollisionIgnoredActors
  (params (HitResults :type Array%20of%20Hit%20Result%20Structures))
  ;[0.1413 0.01328 0.15 0.5] ApplyDamageToDashCollisionIgnoredActors: There are some build actors that are small that we ignore during the dash so we don't stop in unexpected situations. This is where those actors are damaged.
  (if (Utilities|Array|IsNotEmpty HitResults)
    (bind _asyncaction_applydashcollisionignoredamage (Variables|Transient|Async|GetAsyncAction_ApplyDashCollisionIgnoreDamage)
      (:then
        (if (AsyncAction|IsActive _asyncaction_applydashcollisionignoredamage)
          (Ability|Tasks|AddTargets _asyncaction_applydashcollisionignoredamage HitResults)
          (else
            (bind _asynctaskproxy (Ability|Tasks|ApplyGameplayEffectToHitResultsAsync :OriginatingAbilitySystemComponent (Ability|GetAbilitySystemComponentfromActorInfo) :HitResults HitResults :GameplayEffectClass (Variables|GameplayEffects|GetDashCollisionIgnoreDamageEffectClass) :GameplayEffectLevel (General|GetGameplayEffectLevelforDamage) :Stacks 1 :MaxTargetsToProcessPerTickCVar (Utilities|Struct|MakeScalableFloat 3.0)))
            (Variables|Transient|Async|SetAsyncAction_ApplyDashCollisionIgnoreDamage _asynctaskproxy)))
        (for _array_element HitResults
          (Ability|ExecuteGameplayCueWithParamsOnOwner (Utilities|Struct|BreakGameplayCueTag (select (Utilities|Enum|Equal(Enum) (Variables|Transient|GetCurrentDashType) "NewEnumerator0") (Variables|GameplayCueTag|GetDashImpactGameplayCueTag) (Variables|GameplayCueTag|GetSideDashImpactGameplayCueTag))) (GameplayCue|MakeGameplayCueParametersfromHitResult _array_element))))
      (:else
        _))))

(event Custom|HandleDashMovementModeExtEnded
  ;[0 0.15 0.1173 0.5] HandleDashMovementModeExtEnded: End the ability behavior that is running along with the dash.
  (Ability|ApplyGameplayEffectToOwner (Variables|PowerFallImmunity|GetGE_FallDamageImmunity))
  (CallFunction|EndDashAbilityBehavior))

(event Custom|EndDashAbilityBehavior
  ;[0.1045 0.02739 0.15 0.5] EndDashAbilityBehavior: End the dash and jump to the Outro section of the montage.
  (if (not (Variables|Transient|GetbEndDashAbilityBehaviorComplete))
    (Variables|Transient|SetbEndDashAbilityBehaviorComplete true)
    (bind _asynctask_continuoustargeting (Variables|Transient|Async|GetAsyncTask_ContinuousTargeting)
      (:then
        (GameplayTasks|EndTask _asynctask_continuoustargeting))
      (:else
        _))
    (Camera|SetEnableDashCameraModifier false)
    (if (Variables|General|GetbApplyLockoutGameplayEffect)
      (Ability|ApplyGameplayEffectToOwner (Utilities|Select :Index (Variables|Transient|GetCurrentDashType) :NewEnumerator0 (Variables|GameplayEffects|GetForwardDashBlockChangeEquipmentGameplayEffect) :NewEnumerator1 (Variables|GameplayEffects|GetSideDashBlockChangeEquipmentGameplayEffect) :NewEnumerator2 (Variables|GameplayEffects|GetSlamDashBlockChangeEquipmentGameplayEffect)))
      (else
        _))
    (switch Utilities|FlowControl|Switch|SwitchonRiftGauntletsDashTypeEnum (Variables|Transient|GetCurrentDashType)
      (:NewEnumerator0
        (Ability|JumptoMontageForwardPunchSection))
      (:NewEnumerator1
        (Ability|Animation|MontageStop -1.0)
        (GameplayCue|ExecuteRadialForceBurst 0.0)
        (Utilities|FlowControl|DeltaGateWildcard (Variables|Transient|GetbDamageDealt) true
          (:Success
            (bind (_asynctaskproxy _targetdata) (Ability|Tasks|DoTargetSelection "SideDashDamageTargetSelection" "(TagName=\"Targeting.ApplicationTag.1\")")
              (:then
                (Variables|Transient|Async|SetAsyncTask_SideDashDamageTarget _asynctaskproxy))
              (:Targeted
                (bind _output_get (Variables|Transient|SetCachedDamageTargetData _targetdata))
                (Utilities|FlowControl|DelayUntilNextFrame) ; Help perf apply damage after delay.
                (Ability|HandleImpact _output_get (Variables|GameplayEffects|GetSideDashDamageGameplayEffectClass) (Variables|GameplayCueTag|GetSideDashImpactGameplayCueTag))
                (General|ApplyPerformedDamageGameplayEffecttoOwner _output_get))) ; Side dash targeting.
          )))
      (:NewEnumerator2
        (CallFunction|SetPawnCollision (Variables|Transient|GetAvatarPlayerPawn) true)
        (Ability|Animation|MontageJumptoSection (Utilities|String|SelectName "GroundSlam" "AirPunch" (Variables|Transient|GetbLandedFromSlam)))))))

(event EventLaunchPlayer
  ;[0.002618 0.09472 0.15 0.5] LaunchPlayer: Starts the dash movement, notify the camera modifier that the dash is occuring.
  (|CallParentFunction)
  (bind _cameramodifierriftgauntletsdash (Variables|Transient|GetCameraModifierRiftGauntletsDash)
    (:then
      (Default|NotifyDash _cameramodifierriftgauntletsdash))
    (:else
      _))
  (Class|FortMovementModeExtBlueprintLibrary|EndMovementModeExtensiononActor (Variables|Transient|GetHoverMovementModeRuntime) (Variables|Transient|GetAvatarPlayerPawn))
  (Ability|RemoveGameplayEffectFromOwnerWithHandle (Variables|Transient|GetDashIntroGameplayEffectHandle))
  (GameplayCue|ExecuteRadialForceBurst 0.0 0.25 2.6e+03 2.0))

(event Ability|EventOnEndAbility
  (params (bWasCancelled :type Boolean))
  ;[0.1042 0.05442 0 0.5] OnEndAbility: Clean up.
  (Utilities|FlowControl|Sequence
    (:then_0
      (GameplayTasks|EndTask (Utilities|GetAsyncTasks))
      (GameplayTasks|EndTask (Variables|Transient|Async|GetAsyncTask_WaitForSlamLand))
      (bind _asyncaction_damageeffectapplication (Variables|Transient|Async|GetAsyncAction_DamageEffectApplication)
        (:then
          (Ability|Async|EndAction _asyncaction_damageeffectapplication))
        (:else
          _))
      (Camera|SetEnableDashCameraModifier false)
      (MovementMode|StopDashMovementMode)
      (Class|FortMovementModeExtBlueprintLibrary|EndMovementModeExtensiononActor (Variables|Transient|GetHoverMovementModeRuntime) (Variables|Transient|GetAvatarPlayerPawn))
      (Ability|RemoveGameplayEffectFromOwnerWithHandle (Variables|Transient|GetBlockChangeEquipmentEffectHandle))
      (MovementMode|SetEnableHighAccelerationTolerance false)
      (if (Utilities|Enum|Equal(Enum) (Variables|Transient|GetCurrentDashType) "NewEnumerator2")
        (CallFunction|SetPawnCollision (Variables|Transient|GetAvatarPlayerPawn) true)
        (else
          _))
      (bind _returnvalue (Camera|GetCameraAnimationModifier))
      (Utilities|IsValid _returnvalue
        (:Is%20Valid
          (for _array_element (Variables|Transient|GetCameraAnimationHandles)
            (CameraAnimation|StopCameraAnimation _returnvalue _array_element false :!class /Script/EngineCameras.CameraAnimationCameraModifier)))
        (:Is%20Not%20Valid
          _))
      (Animation|SetBindtoMontageSectionChanged false)
      (Ability|RemoveGameplayEffectFromOwnerWithHandle (Variables|Transient|GetDashIntroGameplayEffectHandle))
      (Ability|RemoveGameplayEffectFromOwnerWithHandle (Variables|Transient|GetActiveTagGameplayEffectHandle))
      (Ability|RemoveGameplayEffectFromOwnerWithHandle (Variables|Transient|GetSlamDashPreventMove))
      (Ability|StopMontageBasedonSection)
      (if (Variables|Transient|GetbCommitSuccessful)
        (bind _asb_weapon_rift_gauntlets (Utilities|Casting|CastToB_Weapon_RiftGauntlets (Variables|Transient|GetSourceWeapon)))
        (if (<= (FortWeapon|GetRemainingAmmo _asb_weapon_rift_gauntlets) 0)
          (Class|BWeaponRiftGauntlets|DestroyWeapon _asb_weapon_rift_gauntlets))))
    (:then_1
      (|CallParentFunction bWasCancelled))))

(event EventApplyGuidedMotionMME
  (bind (_x _y) (Math|Vector2D|BreakVector2D (Class|FortMovementBlueprintLibrary|GetLocalMoveInputVector (Variables|Transient|GetAvatarPlayerPawn))))
  ;[0.1069 0.007466 0.15 0.5] ApplyGuidedMotionMME: Based on the circumstance and input, execute the desired MME.
  ;;[0.03302 0.05415 0.15 0.5] Execute the slam ability, if slam is allowed and input is within slam threshold.
  (bind _breturnvalue (General|ShouldExecuteSlam))
  (bind _bvalid (MovementMode|IsInputDirectionWIthinAngleThreshold (Variables|Hotfix|GetHotfix_SlamAllowAngleInput)))
  (if (and _bvalid _breturnvalue)
    (Game|ApplyGuidedMotionMME (Variables|Transient|GetAvatarPlayerPawn) (Variables|PowerLeap|MME|GetGuidedMotionLeapMME) 0 0 (Utilities|Struct|MakeGuidedMotionOptionalValues :OptionalFloats (General|DashTypetoOptionalFloat "NewEnumerator2")))
    (elif (ScalableFloat|GetValueasBool (Variables|Hotfix|GetHotfix_UseMoveToActorWhenTargeted))
      ;;[0.15 0.08736 0.1048 0.5] If slam is not being executed, check to see if there's a pawn that is being targeted.
      (bind _bvalid_2 (MovementMode|IsInputDirectionWIthinAngleThreshold (Variables|Hotfix|GetHotfix_MoveToActorAllowAngleInput)))
      (if _bvalid_2
        (bind _pawn (Targeting|GetTargetingPawn))
        (Utilities|IsValid _pawn
          (:Is%20Valid
            (Game|ApplyGuidedMotionMME (Variables|Transient|GetAvatarPlayerPawn) "/RiftGauntletsGameplay/Gameplay/Dash/MovementMode/FMM_GuidedMotion_RiftGauntlets_Dash_MoveToActor.FMM_GuidedMotion_RiftGauntlets_Dash_MoveToActor_C" _pawn 0 (Utilities|Struct|MakeGuidedMotionOptionalValues :OptionalFloats (General|DashTypetoOptionalFloat "NewEnumerator0"))))
          (:Is%20Not%20Valid
            (if (ScalableFloat|GetValueasBool (Variables|Hotfix|GetHotfix_InputDirectionTargetingEnable))
              (bind (_asynctaskproxy _targetdata) (Ability|Tasks|PerformTargetingAbilityTask "InputDirectionTargetingTask" "/RiftGauntletsGameplay/Gameplay/Dash/TP_RiftGauntlets_Dash_InputDirectionTargeting.TP_RiftGauntlets_Dash_InputDirectionTargeting" false true (ScalableFloat|GetValueasBool (Variables|Hotfix|GetHotfix_InputDirectionTargetingAsync)))
                (:then
                  (Variables|Transient|Async|SetAsyncTask_InputDirectionTargeting _asynctaskproxy))
                (:Targeted
                  (bind _asfort_pawn (Utilities|Casting|CastToFortPawn (.HitActor (Collision|BreakHitResult (Ability|TargetData|GetHitResultfromTargetData _targetdata))))
                    (:then
                      (Game|ApplyGuidedMotionMME (Variables|Transient|GetAvatarPlayerPawn) "/RiftGauntletsGameplay/Gameplay/Dash/MovementMode/FMM_GuidedMotion_RiftGauntlets_Dash_MoveToActor.FMM_GuidedMotion_RiftGauntlets_Dash_MoveToActor_C" _asfort_pawn 0 (Utilities|Struct|MakeGuidedMotionOptionalValues :OptionalFloats (General|DashTypetoOptionalFloat "NewEnumerator0"))))
                    (:CastFailed
                      _)))
                (:Cancelled
                  _))
              ;;[0.02599 0.02118 0.15 0.5] If there's no direct pawn target, check to see if the input is asking for a side dash. Behave based on that check.
              (bind _bvalid_3 (MovementMode|IsInputDirectionWIthinAngleThreshold (Variables|Hotfix|GetHotfix_SideDashAngle)))
              (if _bvalid_3
                _
                (elif (or (> (Math|Float|Absolute(Float) _x) (ScalableFloat|GetValueatLevel (Variables|Hotfix|GetHotfix_SideDashMinInputStrengthRaw))) (> (Math|Float|Absolute(Float) _y) (ScalableFloat|GetValueatLevel (Variables|Hotfix|GetHotfix_SideDashMinInputStrengthRaw))))
                  (Game|ApplyGuidedMotionMME (Variables|Transient|GetAvatarPlayerPawn) (Variables|PowerLeap|MME|GetGuidedMotionLeapMME) 0 0 (Utilities|Struct|MakeGuidedMotionOptionalValues :OptionalFloats (General|DashTypetoOptionalFloat "NewEnumerator1"))) ; The whole chain of sideways dash behavior starts here. On the client's player input side when initiating the MME.
                ))
              (Game|ApplyGuidedMotionMME (Variables|Transient|GetAvatarPlayerPawn) (Variables|PowerLeap|MME|GetGuidedMotionLeapMME) 0 0 (Utilities|Struct|MakeGuidedMotionOptionalValues :OptionalFloats (General|DashTypetoOptionalFloat "NewEnumerator0")))
              (else
                _))))
        (else
          _))
      (else
        _))))

(event EventCastMontageNotifyStart
  (params (NotifyTag :type Gameplay%20Tag%20Structure))
  ;[0.1352 0.0314 0.15 0.5] CastMontageNotifyStart: React to notifies from the montage.
  (|CallParentFunction NotifyTag)
  (if (== NotifyTag (Variables|General|GetDamageAnimNotifyTag))
    (GameplayCue|ExecuteRadialForceBurst 5e+01)
    ;;[0.06874 0.07495 0.15 0.5] Deal the ability's primary damage.
    (Utilities|FlowControl|DeltaGateWildcard (Variables|Transient|GetbDamageDealt) true
      (:Success
        (bind (_asynctaskproxy _targetdata) (Ability|Tasks|PerformTargetingAbilityTask "DamageTargetSelection" (Variables|Targeting|GetDamageTargetingPreset) false false true)
          (:Targeted
            (CallFunction|PerformPrimaryDamage _targetdata)))))
    (elif (== NotifyTag "(TagName=\"RiftGauntlets.AnimNotify.EndAbility.Server\")")
      (Ability|HasAuthority
        (:True
          (Ability|EndAbility)))
      (elif (== NotifyTag "(TagName=\"RiftGauntlets.AnimNotify.Ability.Dash.AllowChangeEquipment\")")
        (Ability|RemoveGameplayEffectFromOwnerWithHandle (Variables|Transient|GetBlockChangeEquipmentEffectHandle))
        (elif (== NotifyTag "(TagName=\"RiftGauntlets.AnimNotify.Ability.Dash.Slam.Impact\")")
          (GameplayCue|ExecuteRadialForceBurst 5e+01 0.45 2.5e+03 1.0)
          ;;[0 0.1447 0.15 0.5] Deal slam impact damage and handle other related behaviors.
          (Utilities|FlowControl|DeltaGateWildcard (Variables|Transient|GetbDamageDealt) true
            (:Success
              (Camera|PlayCameraAnimation (Variables|Camera|GetSlamCameraAnimation))
              (bind _asynctaskproxy_2 (GameplayTasks|TaskWaitDelay 0 (ScalableFloat|GetValueatLevel (Variables|Hotfix|GetHotfix_PostSlamPreventMoveDuration)))
                (:then
                  (Variables|Transient|Async|SetAsyncTask_PostSlamPreventMove _asynctaskproxy_2)
                  (bind (_asynctaskproxy_3 _targetdata_2) (Ability|Tasks|PerformTargetingAbilityTask "LandingTargeting_Pawns" (Variables|Targeting|GetSlamPawnTargetingPreset) false false true)
                    (:Targeted
                      (bind _filteredtargetdata (Targeting|FilterTargetData _targetdata_2))
                      (bind _asynctaskproxy_4 (Ability|Tasks|ApplyGameplayEffectToTargetDataAsync _filteredtargetdata (Variables|GameplayEffects|GetSlamPawnDamageGameplayEffectClass) (General|GetGameplayEffectLevelforDamage) 1 (Utilities|Struct|MakeScalableFloat 3.0)))
                      (Utilities|FlowControl|Sequence
                        (:then_0
                          (General|ApplyPerformedDamageGameplayEffecttoOwner _filteredtargetdata))
                        (:then_1
                          (Utilities|AsyncActionComplete _asynctaskproxy_4
                            (:Complete
                              (bind (_asynctaskproxy_5 _targetdata_3) (Ability|Tasks|PerformTargetingAbilityTask "LandingTargeting_World" (Variables|Targeting|GetSlamWorldTargetingPreset) false false true)
                                (:Targeted
                                  (bind _filteredtargetdata_2 (Targeting|FilterTargetData _targetdata_3))
                                  (Ability|Tasks|ApplyGameplayEffectToTargetDataAsync _filteredtargetdata_2 (Variables|GameplayEffects|GetSlamWorldDamageGameplayEffectClass) (General|GetGameplayEffectLevelforDamage) 1 (Utilities|Struct|MakeScalableFloat 3.0))
                                  (General|ApplyPerformedDamageGameplayEffecttoOwner _filteredtargetdata_2))))) ; Setting up this function cause OnCompleted doesn't seem to get called on the client.
                        )))))
                (:OnFinish
                  (Ability|RemoveGameplayEffectFromOwnerWithHandle (Variables|Transient|GetSlamDashPreventMove)))))))))))

(event Custom|HandlePawnLandedFromSlam
  (params (Hit :type Hit%20Result%20Structure%20%28by%20ref%29))
  ;[0.1103 0.15 0.1317 0.5] HandlePawnLandedFromSlam: Invoked when the slam dash is complete.
  (Variables|Transient|SetbLandedFromSlam true)
  (EventDispatchers|UnbindEventfromOnPawnLanded (Variables|Transient|GetAvatarPlayerPawn) (EventDispatchers|CreateEvent :!event Custom|HandlePawnLandedFromSlam))
  (GameplayTasks|EndTask (Variables|Transient|Async|GetAsyncTask_WaitForSlamLand))
  (MovementMode|StopDashMovementMode)
  (CallFunction|EndDashAbilityBehavior))

(event EventPlayCastAnimation
  ; PlayCastAnimation: Once the animation is played, apply the respective stance gameplay effect and apply the dash intro gameplay effect that the UI team can use.
  (|CallParentFunction)
  (Stance|ApplyStanceGameplayEffect) ; Needs to be set before parent activate ability is called.
  (bind _returnvalue (Ability|ApplyGameplayEffectToOwner (Variables|GameplayEffects|GetDashIntroGameplayEffectClass)))
  (Variables|Transient|SetDashIntroGameplayEffectHandle _returnvalue))

(event Custom|PerformPrimaryDamage
  (params (TargetData :type Gameplay%20Ability%20Target%20Data%20Handle%20Structure))
  ;[0.15 0.02489 0.1035 0.5] PerformPrimaryDamage: Function that applies the primary damage (forward dash and ascend damage in child).
  (bind _filteredtargetdata (Targeting|FilterTargetData TargetData))
  (General|ApplyingPrimaryDamageTypetoTargetData _filteredtargetdata)
  (bind _asyncaction_primaryimpactdamage (Variables|Transient|Async|GetAsyncAction_PrimaryImpactDamage)
    (:then
      (if (AsyncAction|IsActive _asyncaction_primaryimpactdamage)
        (Ability|Tasks|AddTargetData _asyncaction_primaryimpactdamage _filteredtargetdata)
        (else
          (bind _asynctaskproxy (Ability|Tasks|ApplyGameplayEffectToTargetDataAsync _filteredtargetdata (Variables|GameplayEffects|GetDashImpactDamageGameplayEffectClass) (General|GetGameplayEffectLevelforDamage) 1 (Utilities|Struct|MakeScalableFloat 3.0)))
          (Variables|Transient|Async|SetAsyncAction_PrimaryImpactDamage _asynctaskproxy)))
      (for _index (range (Ability|TargetData|GetDataCountfromTargetData _filteredtargetdata))
        (Ability|ExecuteGameplayCueWithParamsOnOwner (Utilities|Struct|BreakGameplayCueTag (Variables|GameplayCueTag|GetDashImpactGameplayCueTag)) (General|MakeGameplayCueParamfromHitResult (Ability|TargetData|GetHitResultfromTargetData _filteredtargetdata _index))))
      (General|ApplyPerformedDamageGameplayEffecttoOwner _filteredtargetdata))
    (:else
      _)))

(fn ApplyingPrimaryDamageTypeToTargetData
  (params (TargetData :type Gameplay%20Ability%20Target%20Data%20Handle%20Structure)))

(fn ApplyPerformedDamageGameplayEffectToOwner
  (params (TargetData :type Gameplay%20Ability%20Target%20Data%20Handle%20Structure))
  (bind _hotfix_rewardchargegrant (Variables|Hotfix|GetHotfix_RewardChargeGrant))
  ;[0.001591 0.1483 0.15 0.5 26] Primarily used for HUD feedback.
  (Utilities|FlowControl|Sequence
    (:then_0
      (if (and (Ability|TargetData|TargetDataHasActor TargetData) (not (Variables|Transient|GetbAppliedPerformedDamageGameplayEffectToOwner)))
        (Ability|ApplyGameplayEffectToOwner (Variables|GameplayEffects|GetPerformedDamageGameplayEffectClass) (Math|Conversions|ToInteger(Boolean) (GameplayTags|HasMatchingGameplayTag (Variables|Transient|GetAvatarPlayerPawn) "(TagName=\"RiftGauntlets.Stance.Dash.Left\")")))
        (Variables|Transient|SetbAppliedPerformedDamageGameplayEffectToOwner true)))
    (:then_1
      (if (ScalableFloat|GetValueasBool _hotfix_rewardchargegrant)
        (Ability|HasAuthority
          (:True
            (if (not (Variables|Transient|GetbRewardChargeGranted))
              (bind _array_element (Utilities|Array|ForEachLoopwithBreak (Ability|TargetData|GetAllActorsfromTargetData TargetData))
                (:LoopBody
                  (bind _asfort_player_pawn (Utilities|Casting|CastToFortPlayerPawn _array_element))
                  (if (not (Utilities|Enum|Equal(Enum) (Team|GetTeamAffiliation _asfort_player_pawn (Variables|Transient|GetAvatarPlayerPawn)) "Friendly"))
                    (bind _sourceweapon (Variables|Transient|GetSourceWeapon)
                      (:then
                        (FortWeapon|ModAmmo _sourceweapon (ScalableFloat|GetValueasInteger _hotfix_rewardchargegrant))
                        (Variables|Transient|SetbRewardChargeGranted true))
                      (:else))))))))))))

(fn ApplyStanceGameplayEffect
  (bind _outfortplayerpawn (CallFunction|GetAvatarasValidFortPlayerPawn)
    (:ValidFortPlayerPawn
      (if (GameplayTags|HasMatchingGameplayTag _outfortplayerpawn "(TagName=\"RiftGauntlets.Stance.Dash.Left\")")
        (Ability|ApplyGameplayEffectToOwner "/RiftGauntletsGameplay/Gameplay/Dash/GE_RiftGauntlets_Dash_LatestStance_Right.GE_RiftGauntlets_Dash_LatestStance_Right_C")
        (else
          (Ability|ApplyGameplayEffectToOwner "/RiftGauntletsGameplay/Gameplay/Dash/GE_RiftGauntlets_Dash_LatestStance_Left.GE_RiftGauntlets_Dash_LatestStance_Left_C"))))) ; Can be called so early the avatar variable is not set, so use the direct function instead.
)

(fn CacheVariables
  ;[0.08191 0.1229 0.15 0.5] Cache variables so they can be used easily.
  (bind _outfortplayerpawn (CallFunction|GetAvatarasValidFortPlayerPawn)
    (:ValidFortPlayerPawn
      (bind _output_get (Variables|Transient|SetAvatarPlayerPawn _outfortplayerpawn))
      (Variables|Transient|SetSourceWeapon (Weapon|GetCurrentWeapon _output_get)))))

(fn DashTypeToOptionalFloat
  (params (Enum :type RiftGauntletsDashTypeEnum%20Enum :default "NewEnumerator0"))
  (return (Utilities|Array|MakeArray (Math|Conversions|ToFloat(Integer) (Math|Conversions|ToInteger(Byte) Enum)))))

(fn ExecuteRadialForceBurst
  (params (ForwardOffset :type Float%20%28single-precision%29) (TimeDecay :type Float%20%28double-precision%29 :default 0.15) (ImpactRadius :type Float%20%28double-precision%29 :default 8e+02) (Strength :type Float%20%28double-precision%29 :default 0.66))
  (Ability|ExecuteGameplayCueWithParamsOnOwner (Utilities|Struct|BreakGameplayCueTag (Variables|GameplayCueTag|GetRadialForceBurstGameplayCueTag)) (Ability|GameplayCue|MakeGameplayCueParameters :NormalizedMagnitude 0.0 :RawMagnitude ForwardOffset :Location (Math|Vector|MakeVector TimeDecay ImpactRadius Strength) :Normal "0,0,0" :GameplayEffectLevel 1 :AbilityLevel 1 :bReplicateLocationWhenUsingMinimalRepProxy false)))

(fn ExitMovementStates
  (bind _avatarplayerpawn (Variables|Transient|GetAvatarPlayerPawn))
  (bind _returnvalue (Vehicle|GetVehicle _avatarplayerpawn))
  (bind _avatarplayerpawn_2 (Variables|Transient|GetAvatarPlayerPawn))
  ;[0.15 0.06547 0.1042 0.5] Exit owner out of the movement modes that they may be in.
  (bind _avatarplayerpawn_3 (Variables|Transient|GetAvatarPlayerPawn)
    (:then
      (Ability|SendGameplayEventtoActor _avatarplayerpawn_3 "(TagName=\"Event.Athena.Balloons.RemoveAll\")") ; Remove any balloons on the player.
      ;;[0.06532 0.01858 0.1302] Exit players that are in the Open seat of a vehicle.
      (if (GameplayTags|HasAnyTags (GameplayTags|GetOwnedGameplayTags _avatarplayerpawn_3 :!class /Script/GameplayTags.BlueprintGameplayTagLibrary) (GameplayTags|MakeLiteralGameplayTagContainer "(GameplayTags=((TagName=\"Gameplay.Status.InVehicle.Seat.Open\")))") true)
        (Utilities|IsValid _avatarplayerpawn
          (:Is%20Valid
            (Collision|AddIgnoredActor _avatarplayerpawn _returnvalue 1.0)
            (Vehicle|EjectPlayerPawn _returnvalue _avatarplayerpawn)))
        (elif (GameplayTags|HasAnyTags (GameplayTags|GetOwnedGameplayTags _avatarplayerpawn_3 :!class /Script/GameplayTags.BlueprintGameplayTagLibrary) (GameplayTags|MakeLiteralGameplayTagContainer "(GameplayTags=((TagName=\"MovementMode.Ziplining\"),(TagName=\"MovementMode.ZipliningOnSpline\")))") true)
          ;;[0.05208 0.008721 0.05085] End ziplining.
          (Ability|HasAuthority
            (:True
              (bind _returnvalue_2 (Zipline|GetActiveZipline _avatarplayerpawn_2))
              (Utilities|IsValid _returnvalue_2
                (:Is%20Valid
                  (Class|FortAthenaZiplineBase|AuthorityForceEndZiplining _returnvalue_2 _avatarplayerpawn_2))))
            (:False
              (Zipline|EndZiplining _avatarplayerpawn_2 false false))))))
    (:else)))

(fn FilterTargetData
  (params (TargetData :type Gameplay%20Ability%20Target%20Data%20Handle%20Structure))
  ; FilterTargetData: Removing duplicate actor entries.
  (for _index (range (Ability|TargetData|GetDataCountfromTargetData TargetData))
    (Utilities|IsValid (.HitActor (Collision|BreakHitResult (Ability|TargetData|GetHitResultfromTargetData TargetData _index)))
      (:Is%20Valid
        (bind _returnvalue (Utilities|Array|AddUnique (Variables|Transient|GetDashHitActors) (.HitActor (Collision|BreakHitResult (Ability|TargetData|GetHitResultfromTargetData TargetData _index)))))
        (if (>= _returnvalue 0)
          (bind _returnvalue_2 (Ability|TargetData|AppendTargetDataHandle (LocalVariables|GetReturnHandle) (Ability|TargetData|AbilityTargetDatafromHitResult (Ability|TargetData|GetHitResultfromTargetData TargetData _index))))
          (LocalVariables|SetReturnHandle _returnvalue_2)))))
  (return (LocalVariables|GetReturnHandle)))

(fn GetCameraAnimationModifier
  (Utilities|FlowControl|Sequence
    (:then_0
      (bind _avatarplayerpawn (Variables|Transient|GetAvatarPlayerPawn)
        (:then
          (Utilities|IsValid (Fortnite|GetFortPlayerControllerfromActor _avatarplayerpawn)
            (:Is%20Valid
              (bind _playercameramanager (Variables|/Script/Engine.PlayerController|GetPlayerCameraManager (Fortnite|GetFortPlayerControllerfromActor _avatarplayerpawn))
                (:then
                  (return (CameraAnimation|ConvCameraAnimationCameraModifier _playercameramanager)))
                (:else)))))
        (:else)))
    (:then_1
      (return 0))))

(fn GetCastMontage
  (bind _returnvalue (CallFunction|GetAvatarasFortPlayerPawn))
  (Utilities|FlowControl|Sequence
    (:then_0
      (Utilities|IsValid _returnvalue
        (:Is%20Valid
          (if (GameplayTags|HasMatchingGameplayTag _returnvalue "(TagName=\"RiftGauntlets.Stance.Dash.Right\")")
            (return (Variables|PowerAnimationCast|GetCastMontage))
            (elif (GameplayTags|HasMatchingGameplayTag _returnvalue "(TagName=\"RiftGauntlets.Stance.Dash.Left\")")
              (return (Variables|Animation|GetCastMontageFlipped))
              (else
                (return (Variables|PowerAnimationCast|GetCastMontage))))))))
    (:then_1
      (return (|CallParentFunction)))))

(fn GetCustomAbilitySourceTransform
  (bind _returnvalue (CallFunction|GetAvatarasFortPlayerPawn))
  (bind (_returnvalue_location _returnvalue_rotation _returnvalue_scale) (Math|Transform|BreakTransform (Transformation|GetActorTransform _returnvalue)))
  (Utilities|IsValid _returnvalue
    (:Is%20Valid
      (return _returnvalue_location (Math|Rotator|SelectRotator (Math|Rotator|MakeRotfromX (Transformation|GetVelocity _returnvalue)) (Transformation|GetActorRotation _returnvalue) (> (Math|Vector|VectorLength (Transformation|GetVelocity _returnvalue)) 0)) _returnvalue_scale))
    (:Is%20Not%20Valid
      ;[1 0 0] Warning: Avatar player pawn is NULL.
      (Fortnite|FortShippingLog self "GetCustomAbilitySourceTransform: Avatar player pawn is NULL." true)
      (return "0, 0, 0" "0, 0, 0" "1.000000,1.000000,1.000000"))))

(fn GetGameplayEffectLevelForDamage
  (bind _returnvalue (Ability|GetCurrentSourceWeapon))
  ;[0.1106 0.01642 0.15 0.5] Return 1 for the mythic weapon.
  (Utilities|FlowControl|Sequence
    (:then_0
      (Utilities|IsValid _returnvalue
        (:Is%20Valid
          (if (Utilities|Enum|Equal(Enum) (Rarity|GetWeaponDataRarity _returnvalue) "Mythic")
            (return 1)))))
    (:then_1
      (return 0))))

(fn GetTargetingPawn
  (bind _returnvalue (Ability|GetCurrentSourceWeapon))
  (bind _returnvalue_2 (CallFunction|GetAvatarasFortPlayerPawn))
  (bind (_bblockinghit _binitialoverlap _time _distance _location _impactpoint _normal _impactnormal _physmat _hitactor) (Collision|BreakHitResult (Class|BAbilityTargetingFeedbackComponentParent|GetLatestTargetingTraceHitResult (Actor|GetComponentbyClass _returnvalue "/AbilityTargetingFeedbackGameplay/Gameplay/B_AbilityTargetingFeedbackComponent_Parent.B_AbilityTargetingFeedbackComponent_Parent_C"))))
  (bind _returnvalue_3 (Actor|GetComponentbyClass _returnvalue "/AbilityTargetingFeedbackGameplay/Gameplay/B_AbilityTargetingFeedbackComponent_Parent.B_AbilityTargetingFeedbackComponent_Parent_C"))
  (Utilities|FlowControl|Sequence
    (:then_0
      (Utilities|IsValid (AI|GetAIController _returnvalue_2)
        (:Is%20Valid
          (bind _targetactor (Class|BPFLAICommonFunctions|GetTargetActorfromBlackboard _returnvalue_2))
          (return _targetactor))))
    (:then_1
      (Utilities|IsValid _returnvalue
        (:Is%20Valid
          (Utilities|IsValid _returnvalue_3
            (:Is%20Valid
              (if (Utilities|Enum|Equal(Enum) (Class|BAbilityTargetingFeedbackComponentParent|GetLatestTargetingActorType _returnvalue_3) "NewEnumerator2")
                (bind _asfort_pawn (Utilities|Casting|CastToFortPawn _hitactor))
                (return _asfort_pawn)))))))
    (:then_2
      (return 0))))

(fn HandleImpact
  (params (TargetData :type Gameplay%20Ability%20Target%20Data%20Handle%20Structure) (DamageGameplayEffectClass :type Gameplay%20Effect%20Class%20Reference) (GameplayCueTag :type Gameplay%20Cue%20Tag%20Structure%20%28by%20ref%29))
  (bind _filteredtargetdata (Targeting|FilterTargetData TargetData))
  (if (Ability|TargetData|TargetDataHasActor _filteredtargetdata)
    (Ability|ApplyGameplayEffectToTarget _filteredtargetdata DamageGameplayEffectClass (General|GetGameplayEffectLevelforDamage))
    (Ability|ExecuteGameplayCueWithParamsOnOwner (Utilities|Struct|BreakGameplayCueTag GameplayCueTag) (General|MakeGameplayCueParamfromHitResult (Ability|TargetData|GetHitResultfromTargetData _filteredtargetdata)))))

(fn HandleMontageSectionChanged
  (params (Montage :type Anim%20Montage%20Object%20Reference) (SectionName :type Name) (bLooped :type Boolean))
  ;[0.03047 0.135 0.15 0.5] HandleMontageSectionChanged: Update the montage section based on the hotfix value.
  (if (or (== Montage (Variables|Animation|GetCastMontageFlipped)) (== Montage (Variables|PowerAnimationCast|GetCastMontage)))
    (bind _avatarplayerpawn (Variables|Transient|GetAvatarPlayerPawn)
      (:then
        (bind _mesh (Variables|/Script/Engine.Character|GetMesh _avatarplayerpawn)
          (:then
            (Animation|Montage|MontageSetPlayRate (Components|SkeletalMesh|GetAnimInstance _mesh) Montage (select (== SectionName "Cast") (ScalableFloat|GetValueatLevel (Variables|Hotfix|GetHotfix_CastSectionMontagePlayRate)) (Math|Float|SelectFloat (ScalableFloat|GetValueatLevel (Variables|Hotfix|GetHotfix_ForwardIntroSectionMontagePlayRate)) 1.0 (== SectionName "ForwardIntro")))))
          (:else)))
      (:else))))

(fn IsInputDirectionWIthinAngleThreshold
  (params (Angle :type Scalable%20Float%20Structure%20%28by%20ref%29))
  ;[0.04325 0.02377 0.15 0.5] IsInputDirectionWIthinAngleThreshold: Returns if the input direction is below angle threshold.
  (Utilities|FlowControl|Sequence
    (:then_0
      (bind _avatarplayerpawn (Variables|Transient|GetAvatarPlayerPawn)
        (:then
          (Utilities|IsValid (AI|GetAIController (CallFunction|GetAvatarasFortPlayerPawn))
            (:Is%20Valid
              (return true))
            (:Is%20Not%20Valid
              (bind _ascharacter_movement_component (Utilities|Casting|CastToCharacterMovementComponent (Pawn|GetMovementComponent _avatarplayerpawn)))
              (bind _returnvalue (Fortnite|GetAngleDegrees (Transformation|GetActorForwardVector _avatarplayerpawn) (Pawn|Components|CharacterMovement|GetCurrentAcceleration _ascharacter_movement_component)))
              (return (not (and (> _returnvalue (ScalableFloat|GetValueatLevel Angle)) (not (Math|Vector|VectorIsNearlyZero (Pawn|Components|CharacterMovement|GetCurrentAcceleration _ascharacter_movement_component)))))))))
        (:else)))
    (:then_1
      (return false))))

(fn JumpToMontageForwardPunchSection
  (if (Variables|General|GetbJumpToPunchMontageSection)
    (if (not (Variables|Transient|GetbJumpToMontageForwardPunchSectionComplete))
      (Variables|Transient|SetbJumpToMontageForwardPunchSectionComplete true)
      (bind _returnvalue (General|TraceClosetoGround))
      (Ability|Animation|MontageJumptoSection (select _returnvalue "GroundPunch" "AirPunch")))))

(fn CanActivateAbility
  (params (ActorInfo :type Gameplay%20Ability%20Actor%20Info%20Structure) (Handle :type Gameplay%20Ability%20Spec%20Handle%20Structure))
  (bind _returnvalue (Ability|GetActivatingPawn))
  (Utilities|FlowControl|Sequence
    (:then_0
      (if (ScalableFloat|GetValueasBool (Variables|Hotfix|GetHotfix_CheckHolsterForActivate))
        (Utilities|IsValid _returnvalue
          (:Is%20Valid
            ; Grinding can't handle holster check from server, need to skip
            (if (GameplayTags|HasMatchingGameplayTag _returnvalue "(TagName=\"Granted.Athena.GrindRails.IsGrinding\")")
              _
              (elif (Weapon|IsWeaponHolstered _returnvalue)
                (return 0 false)))))))
    (:then_1
      (bind (_relevanttags _returnvalue_2) (|CallParentFunction ActorInfo Handle))
      (return _relevanttags _returnvalue_2))))

(fn OverrideFailedReason
  (params (FailedReason :type Gameplay%20Tag%20Container%20Structure%20%28by%20ref%29))
  (bind _returnvalue (Ability|GetActivatingPawn))
  (bind _returnvalue_2 (Weapon|GetCurrentWeapon _returnvalue))
  (bind _returnfailedreason (LocalVariables|GetReturnFailedReason))
  (bind _overridenfailedreason (|CallParentFunction FailedReason))
  (LocalVariables|SetReturnFailedReason _overridenfailedreason)
  (Utilities|FlowControl|Sequence
    (:then_0
      (bind _returnvalue_3 (Utilities|Time|GetGameTimeinSeconds))
      (if (> (- _returnvalue_3 (Variables|Transient|GetLastPlayedAbilityCooldownGameTime)) (Variables|Sound|GetPlayAbilityOnCooldownSoundInterval))
        (bind _breturn (Sound|ShouldPlayAbilityonCooldownSound FailedReason))
        (if _breturn
          (Variables|Transient|SetLastPlayedAbilityCooldownGameTime _returnvalue_3)
          (Audio|PlaySound2D (Variables|Sound|GetAbilityOnCooldownSound)))))
    (:then_1
      (if (GameplayTags|HasTag _returnfailedreason "(TagName=\"Activate.Failed.CanActivateAbility\")" true)
        (Utilities|IsValid _returnvalue
          (:Is%20Valid
            (Utilities|IsValid _returnvalue_2
              (:Is%20Valid
                (if (<= (FortWeapon|GetMagazineAmmoCount _returnvalue_2) 0)
                  (GameplayTags|AddGameplayTag (LocalVariables|GetReturnFailedReason) "(TagName=\"Athena.Ability.Failed.RiftGauntlets.Dash.NotEnoughCharge\")" :!class /Script/GameplayTags.BlueprintGameplayTagLibrary))))))
        (elif (GameplayTags|HasTag _returnfailedreason "(TagName=\"Activate.Failed.Cooldown\")")
          (GameplayTags|AddGameplayTag (LocalVariables|GetReturnFailedReason) "(TagName=\"Athena.Ability.Failed.NoMessage\")" :!class /Script/GameplayTags.BlueprintGameplayTagLibrary))))
    (:then_2
      (return (LocalVariables|GetReturnFailedReason)))))

(fn MakeGameplayCueParamFromHitResult
  (params (HitResult :type Hit%20Result%20Structure%20%28by%20ref%29))
  (bind (_normalizedmagnitude _rawmagnitude _effectcontext _matchedtagname _originaltag _aggregatedsourcetags _aggregatedtargettags _location _normal _instigator _effectcauser _sourceobject _physicalmaterial _gameplayeffectlevel _abilitylevel _targetattachcomponent _breplicatelocationwhenusingminimalrepproxy) (Ability|GameplayCue|BreakGameplayCueParameters (GameplayCue|MakeGameplayCueParametersfromHitResult HitResult)))
  (bind _output_get (LocalVariables|SetEffectContext (Class|AthenaFunctionLibrary|MakeSimpleFortEffectContext (CallFunction|GetAvatarasFortPlayerPawn) (Ability|GetCurrentSourceWeapon))))
  (Ability|EffectContext|AddHitResult _output_get HitResult)
  (return (Ability|GameplayCue|MakeGameplayCueParameters _normalizedmagnitude _rawmagnitude _output_get _matchedtagname _originaltag _aggregatedsourcetags _aggregatedtargettags _location _normal _instigator _effectcauser _sourceobject _physicalmaterial _gameplayeffectlevel _abilitylevel _targetattachcomponent _breplicatelocationwhenusingminimalrepproxy)))

(fn OnLaunchStarted
  (|CallParentFunction)
  (Utilities|IsValid (AI|GetAIController (CallFunction|GetAvatarasFortPlayerPawn))
    (:Is%20Valid
      (CallFunction|ApplyGuidedMotionMME))))

(fn OptionalFloatToDashType
  (params (Float :type Float%20%28double-precision%29))
  (return (Utilities|Enum|BytetoEnumRiftGauntletsDashTypeEnum (Math|Conversions|ToByte(Integer) (Math|Float|Truncate Float)))))

(fn PlayCameraAnimation
  (params (CameraAnimation :type Camera%20Animation%20Sequence%20Object%20Reference))
  (bind _returnvalue (Camera|GetCameraAnimationModifier))
  (Utilities|IsValid _returnvalue
    (:Is%20Valid
      (bind _returnvalue_2 (CameraAnimation|PlayCameraAnimation _returnvalue CameraAnimation (Utilities|Struct|MakeCameraAnimationParams 1.0 1.0 "Linear" 0.2 "Linear" 1.0) :!class /Script/EngineCameras.CameraAnimationCameraModifier))
      (Utilities|Array|Add (Variables|Transient|GetCameraAnimationHandles) _returnvalue_2)
      (return))))

(fn ProcessContinuousTargetingTaskTargetData
  (params (TargetData :type Gameplay%20Ability%20Target%20Data%20Handle%20Structure%20%28by%20ref%29))
  ;[0.03371 0.1371 0.15 0.5] ProcessContinuousTargetingTaskTargetData: Based on the actor detected decide if we should ignore it so we're able to dash through it or not.
  (if (not (ScalableFloat|GetValueasBool (Variables|Hotfix|GetHotfix_UseSimplerContinuousTargetDetectReactBehavior)))
    (Utilities|FlowControl|Sequence
      (:then_0
        ;;[0.1319 0.04658 0.15 0.5] If we detect a pawn, we should stop immediately and land a blow.
        (bind _returnvalue (Ability|TargetData|GetAllActorsfromTargetData TargetData))
        (for _array_element _returnvalue
          (if (Math|Boolean|ORBoolean (.bSuccess (Utilities|Casting|CastToFortPawn _array_element)) (.bSuccess (Utilities|Casting|CastToFortAIPawn _array_element)) (.bSuccess (Utilities|Casting|CastToFortAthenaVehicle _array_element)))
            (return true))))
      (:then_1
        ;;[0.15 0.02549 0.1441 0.5] Loop through detected builds to see if we hit anything large enough that we need to stop for.
        (for _index (range (Ability|TargetData|GetDataCountfromTargetData TargetData))
          (bind _returnvalue_2 (Ability|TargetData|GetHitResultfromTargetData TargetData _index))
          (bind _asbuilding_actor (Utilities|Casting|CastToBuildingActor (.HitActor (Collision|BreakHitResult _returnvalue_2))))
          (if (Variables|/Script/Engine.Actor|GetbCanBeDamaged _asbuilding_actor)
            (if (GameplayTags|HasMatchingGameplayTag _asbuilding_actor "(TagName=\"Building.Type.Prop\")")
              (if (< (Math|Vector|VectorLength (.BoxExtent (Collision|GetActorBounds _asbuilding_actor true false))) (ScalableFloat|GetValueatLevel (Variables|Hotfix|GetHotfix_IgnoreBuildingsBoundSizeForContinuousDash)))
                (Development|PrintString (Utilities|String|Append "ProcessContinuousTargetingTaskTargetData: Building actor: " (Utilities|GetDisplayName _asbuilding_actor) " :BoxExtent: " (Utilities|String|ToString(Float) (Math|Vector|VectorLength (.BoxExtent (Collision|GetActorBounds _asbuilding_actor true false))))))
                (bind _avatarplayerpawn (Variables|Transient|GetAvatarPlayerPawn)
                  (:then
                    (bind _returnvalue_3 (Utilities|Array|AddUnique (Variables|Transient|GetDashCollisionIgnoredActors) _asbuilding_actor))
                    (if (>= _returnvalue_3 0)
                      (Collision|AddIgnoredActor _avatarplayerpawn _asbuilding_actor (ScalableFloat|GetValueatLevel (Variables|Hotfix|GetHotfix_IgnoreSmallBuildDuration)))
                      ;;; Do not double damage.
                      (bind _returnvalue_4 (Utilities|Array|AddUnique (Variables|Transient|GetDashHitActors) _asbuilding_actor))
                      (if (>= _returnvalue_4 0)
                        (Utilities|Array|Add (LocalVariables|GetDashCollisionIgnoreHitResultsToDamage) _returnvalue_2))))
                  (:else))
                (elif (> (- (Utilities|Time|GetGameTimeinSeconds) (Variables|Transient|GetDashStartGameTime)) (ScalableFloat|GetValueatLevel (Variables|Hotfix|GetHotfix_DashDurationCollisionDetectNoStopTime)))
                  (LocalVariables|SetbReturnStopDash true))))))
        ;;[0.1485 0.007081 0.15 0.5] Apply damage to build actors that we have ignored cause they were too small.
        (CallFunction|ApplyDamagetoDashCollisionIgnoredActors (LocalVariables|GetDashCollisionIgnoreHitResultsToDamage))
        (return (LocalVariables|GetbReturnStopDash))))
    (else
      (return (Ability|TargetData|TargetDataHasHitResult TargetData)))))

(fn ProcessDamageSelectedTargetData
  (params (TargetData :type Gameplay%20Ability%20Target%20Data%20Handle%20Structure)))

(fn ResetVariables
  (Utilities|Array|Clear (Variables|Transient|GetDashHitActors))
  (Variables|Transient|SetbEndDashAbilityBehaviorComplete false)
  (Variables|Transient|SetCurrentDashType "NewEnumerator0")
  (Variables|Transient|SetbJumpToMontageForwardPunchSectionComplete false)
  (Variables|Transient|SetCachedDamageTargetData)
  (Variables|Transient|SetbLandedFromSlam false)
  (Variables|Transient|SetbAppliedPerformedDamageGameplayEffectToOwner false)
  (Variables|Transient|SetbDamageDealt false)
  (Utilities|Array|Clear (Variables|Transient|GetDashCollisionIgnoredActors))
  (Variables|Transient|SetbRewardChargeGranted false))

(fn SetBindToMontageSectionChanged
  (params (bBind :type Boolean))
  (bind _outputdelegate (EventDispatchers|CreateEvent :!event HandleMontageSectionChanged))
  ;[0.04609 0.15 0.1124 0.5] SetBindToMontageSectionChanged: Bind to know whenever the montage section changes. Also, unbind if asked.
  (bind _avatarplayerpawn (Variables|Transient|GetAvatarPlayerPawn)
    (:then
      (bind _mesh (Variables|/Script/Engine.Character|GetMesh _avatarplayerpawn)
        (:then
          (Utilities|IsValid (Components|SkeletalMesh|GetAnimInstance _mesh)
            (:Is%20Valid
              (if bBind
                (EventDispatchers|BindEventtoOnMontageSectionChanged (Components|SkeletalMesh|GetAnimInstance _mesh) _outputdelegate)
                (Animation|HandleMontageSectionChanged (Animation|Montage|GetCurrentActiveMontage (Components|SkeletalMesh|GetAnimInstance _mesh)) (Animation|Montage|MontageGetCurrentSection (Components|SkeletalMesh|GetAnimInstance _mesh) 0))
                (else
                  (EventDispatchers|UnbindEventfromOnMontageSectionChanged (Components|SkeletalMesh|GetAnimInstance _mesh) _outputdelegate))))))
        (:else)))
    (:else)))

(fn SetEnableDashCameraModifier
  (params (bEnable :type Boolean))
  (bind _dashcameramodifierclass (Variables|Camera|GetDashCameraModifierClass))
  ;[0.000999 0.15 0.06155 0.5] SetEnableMantisCameraModifier: Add the camera modifier for that will be used to adjust the camera during the melee attacks.
  (bind _avatarplayerpawn (Variables|Transient|GetAvatarPlayerPawn)
    (:then
      (Utilities|IsValid (Fortnite|GetFortPlayerControllerfromActor _avatarplayerpawn)
        (:Is%20Valid
          (if bEnable
            (bind _playercameramanager (Variables|/Script/Engine.PlayerController|GetPlayerCameraManager (Fortnite|GetFortPlayerControllerfromActor _avatarplayerpawn))
              (:then
                (bind _returnvalue (Game|Player|FindCameraModifierbyClass _playercameramanager _dashcameramodifierclass))
                (Utilities|IsValid _returnvalue
                  (:Is%20Valid
                    (bind _asfort_camera_modifier_custom (Utilities|Casting|CastToFortCameraModifier_Custom _returnvalue))
                    (if (or (CameraModifier|IsDisabled _asfort_camera_modifier_custom) (CameraModifier|IsPendingDisable _asfort_camera_modifier_custom))
                      _))
                  (:Is%20Not%20Valid
                    _))
                (bind _returnvalue_2 (Game|Player|AddNewCameraModifier _playercameramanager _dashcameramodifierclass))
                (bind _ascamera_modifier_rift_gauntlets_parent (Utilities|Casting|CastToCameraModifier_RiftGauntlets_Parent _returnvalue_2))
                (Variables|Transient|SetCameraModifierRiftGauntletsDash _ascamera_modifier_rift_gauntlets_parent))
              (:else))
            (else
              (bind _playercameramanager_2 (Variables|/Script/Engine.PlayerController|GetPlayerCameraManager (Fortnite|GetFortPlayerControllerfromActor _avatarplayerpawn))
                (:then
                  (bind _returnvalue_3 (Game|Player|FindCameraModifierbyClass _playercameramanager_2 _dashcameramodifierclass))
                  (bind _ascamera_modifier_rift_gauntlets_parent_2 (Utilities|Casting|CastToCameraModifier_RiftGauntlets_Parent _returnvalue_3))
                  (Class|CameraModifierRiftGauntletsParent|DisableandRemove _ascamera_modifier_rift_gauntlets_parent_2))
                (:else)))))))
    (:else)))

(fn SetEnableHighAccelerationTolerance
  (params (bEnable :type Boolean))
  (bind _avatarplayerpawn (Variables|Transient|GetAvatarPlayerPawn)
    (:then
      (bind _asfort_movement_comp_character (Utilities|Casting|CastToFortMovementComp_Character (Variables|/Script/Engine.Character|GetCharacterMovement _avatarplayerpawn)))
      (NetTolerance|SetUseHighAccelerationTolerance _asfort_movement_comp_character bEnable "RiftGauntlet"))
    (:else)))

(fn SetPawnCollision
  (params (Character :type Character%20Object%20Reference) (CollisionOn :type Boolean))
  (if (ScalableFloat|GetValueasBool (Variables|Hotfix|GetHotfix_AllowPawnCollisionManipulation))
    (|CallParentFunction Character CollisionOn)))

(fn ShouldExecuteSlam
  (Utilities|FlowControl|Sequence
    (:then_0
      (bind _avatarplayerpawn (Variables|Transient|GetAvatarPlayerPawn)
        (:then
          (return (GameplayTags|HasMatchingGameplayTag _avatarplayerpawn "(TagName=\"RiftGauntlets.Intention.Dash.Slam\")")))
        (:else)))
    (:then_1
      (return false))))

(fn ShouldPlayAbilityOnCooldownSound
  (params (FailedReason :type Gameplay%20Tag%20Container%20Structure))
  (return (not (GameplayTags|HasTag FailedReason "(TagName=\"Activate.Failed.Cooldown\")"))))

(fn ShouldPlayFailedMontage
  (params (FailedReason :type Gameplay%20Tag%20Container%20Structure%20%28by%20ref%29))
  (Utilities|FlowControl|Sequence
    (:then_0
      (bind _returnvalue (CallFunction|GetAvatarasFortPlayerPawn))
      (Utilities|IsValid _returnvalue
        (:Is%20Valid
          (Utilities|IsValid (Animation|GetCurrentMontage _returnvalue :!class /Script/Engine.Character)
            (:Is%20Valid
              (return false))))))
    (:then_1
      (if (GameplayTags|HasTag FailedReason "(TagName=\"Activate.Failed.Cooldown\")")
        (return false)
        (else
          (return true))))))

(fn StopDashMovementMode
  (bind _avatarplayerpawn (Variables|Transient|GetAvatarPlayerPawn)
    (:then
      (bind _charactermovement (Variables|/Script/Engine.Character|GetCharacterMovement _avatarplayerpawn)
        (:then
          (bind _asfort_movement_comp_character (Utilities|Casting|CastToFortMovementComp_Character _charactermovement))
          (Class|FortMovementCompCharacter|CancelActiveExtensionwithTag _asfort_movement_comp_character (Variables|MovementMode|GetDashMovementModeTagContainer)))
        (:else)))
    (:else)))

(fn StopMontageBasedOnSection
  ;[0.08268 0.06697 0.15 0.5] Stop the montage if its in a section that leads into a loop.
  (bind _avatarplayerpawn (Variables|Transient|GetAvatarPlayerPawn)
    (:then
      (bind _mesh (Variables|/Script/Engine.Character|GetMesh _avatarplayerpawn)
        (:then
          (Utilities|IsValid (Components|SkeletalMesh|GetAnimInstance _mesh)
            (:Is%20Valid
              (for _array_element (Utilities|Array|MakeArray (Variables|Animation|GetCastMontageFlipped) (Variables|PowerAnimationCast|GetCastMontage))
                (if (Math|Boolean|ORBoolean (== (Animation|Montage|MontageGetCurrentSection (Components|SkeletalMesh|GetAnimInstance _mesh) _array_element) "Loop") (== (Animation|Montage|MontageGetCurrentSection (Components|SkeletalMesh|GetAnimInstance _mesh) _array_element) "ForwardIntro") (== (Animation|Montage|MontageGetCurrentSection (Components|SkeletalMesh|GetAnimInstance _mesh) _array_element) "Cast"))
                  (Animation|Montage|MontageStop (Components|SkeletalMesh|GetAnimInstance _mesh) 0.3 _array_element))))))
        (:else)))
    (:else)))

(fn TraceCloseToGround
  ;[0.05596 0.1641 0.2083] Returns if the player is close to the ground.
  (Utilities|FlowControl|Sequence
    (:then_0
      (bind _avatarplayerpawn (Variables|Transient|GetAvatarPlayerPawn)
        (:then
          (bind _capsulecomponent (Variables|/Script/Engine.Character|GetCapsuleComponent _avatarplayerpawn)
            (:then
              (LocalVariables|SetCapsuleComponent _capsulecomponent)
              (if (ScalableFloat|GetValueasBool (Variables|Hotfix|GetHotfix_UseCapsuleTraceBeforeGroundDistanceCheck))
                (if (not (Variables|Transient|GetbEndDashAbilityBehaviorComplete))
                  (bind _outhit (Collision|CapsuleTraceByChannel (Transformation|GetWorldLocation (LocalVariables|GetCapsuleComponent)) (+ (Transformation|GetWorldLocation (LocalVariables|GetCapsuleComponent)) (* (Transformation|GetVelocity (Variables|Transient|GetAvatarPlayerPawn)) (Math|Float|Max(Float) (- (ScalableFloat|GetValueatLevel (Variables|Hotfix|GetHotfix_DashDuration)) (- (Utilities|Time|GetGameTimeinSeconds) (Variables|Transient|GetDashStartGameTime)))))) (Components|Capsule|GetScaledCapsuleRadius (LocalVariables|GetCapsuleComponent)) (Components|Capsule|GetScaledCapsuleHalfHeight (LocalVariables|GetCapsuleComponent)) "TraceTypeQuery10" true (Utilities|Array|MakeArray (Variables|Transient|GetAvatarPlayerPawn)) "None" true "(R=1.000000,G=0.957498,B=0.000000,A=1.000000)"))
                  (LocalVariables|SetTraceStartLocation (Math|Vector|SelectVector (.Location (Collision|BreakHitResult _outhit)) (.TraceEnd (Collision|BreakHitResult _outhit)) (Collision|BreakHitResult _outhit)))
                  (else
                    (LocalVariables|SetTraceStartLocation (Transformation|GetActorLocation (Variables|Transient|GetAvatarPlayerPawn)))))
                (LocalVariables|SetTraceEndLocation (Utilities|Operators|Subtract (LocalVariables|GetTraceStartLocation) (Math|Vector|MakeVector 0.0 0.0 (+ (Components|Capsule|GetScaledCapsuleHalfHeight (LocalVariables|GetCapsuleComponent)) (ScalableFloat|GetValueatLevel (Variables|Hotfix|GetHotfix_GroundDistanceCheck))))))
                (bind (_outhit_2 _returnvalue) (Collision|LineTraceByChannel (LocalVariables|GetTraceStartLocation) (LocalVariables|GetTraceEndLocation) "TraceTypeQuery10" true (Utilities|Array|MakeArray (Variables|Transient|GetAvatarPlayerPawn)) "None" true "(R=1.000000,G=0.032781,B=0.927271,A=1.000000)" "(R=0.000000,G=1.000000,B=0.000000,A=1.000000)" 1e+01))
                (return _returnvalue)
                (else
                  _)))
            (:else)))
        (:else)))
    (:then_1
      (return false))))
