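; Passive Rift Gauntlets ability that, on non-authority clients, detects slam intent and retriggers Dash or Ascend if its input is held when cooldown ends.
; Gameplay tag queries on falling toggle a next-tick loop that checks ammo, downward aim angle and a camera trace.
; A hit adds a loose slam-intent tag for dash and HUD use and moves an owner-only telegraph actor, hidden while dashing.
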(event Ability|EventActivateAbility
  (bind _returnvalue (GameplayTags|MakeGameplayTagQuery "(TokenStreamVersion=0,TagDictionary=((TagName=\"MovementMode.Falling\")),QueryTokenStream=(0,1,5,1,2,1,0),UserDescription=\"\",AutoDescription=\" ALL(  ALL( MovementMode.Falling ) )\")"))
  ;[0.0195 0.15 0.02916 0.5] ActivateAbility
  (|CallParentFunction)
  (Ability|HasAuthority
    (:False
      (bind _outfortplayerpawn (CallFunction|GetAvatarasValidFortPlayerPawn)
        (:ValidFortPlayerPawn
          (Variables|Transient|SetAvatarPlayerPawn _outfortplayerpawn)
          ;;[0.1485 0.05623 0.15 0.5] When falling, start to detect if the player would want to slam.
          (bind _asynctaskproxy (Ability|Tasks|WaitGameplayTagQuery _returnvalue)
            (:then
              (Utilities|Array|Add (Variables|Transient|GetAsyncTasks) _asynctaskproxy)
              (bind _asynctaskproxy_2 (Ability|Tasks|WaitGameplayTagQuery _returnvalue 0 "WhenFalse")
                (:then
                  (Utilities|Array|Add (Variables|Transient|GetAsyncTasks) _asynctaskproxy_2)
                  ;[0.02009 0.03324 0.15 0.5] Wait for the ascend cooldown to end.
                  (bind _asynctaskproxy_3 (Ability|Tasks|WaitGameplayTagRemove "(TagName=\"Granted.Cooldown.Athena.RiftGauntlets.Ascend\")")
                    (:then
                      (Variables|Transient|SetAsyncTask_WaitCooldownEndAscend _asynctaskproxy_3)
                      (bind _asynctaskproxy_4 (Ability|Tasks|WaitGameplayTagRemove "(TagName=\"Granted.Cooldown.Athena.RiftGauntlets.Dash\")")
                        (:then
                          (Variables|Transient|SetAsyncTask_WaitCooldownEndDash _asynctaskproxy_4))
                        (:Removed
                          (bind _outfortplayerpawn_2 (CallFunction|GetAvatarasValidFortPlayerPawn)
                            (:ValidFortPlayerPawn
                              (if (Variables|/Script/FortniteGame.FortPawn|GetbPrimaryInputHeld _outfortplayerpawn_2)
                                (Utilities|IsValid (Ability|GetFortAbilitySystemComponent _outfortplayerpawn_2)
                                  (:Is%20Valid
                                    (Abilities|TryActivateAbilitiesbyTag (Ability|GetFortAbilitySystemComponent _outfortplayerpawn_2) (GameplayTags|MakeLiteralGameplayTagContainer "(GameplayTags=((TagName=\"RiftGauntlets.Ability.Dash\")))") true)))))))))
                    (:Removed
                      (bind _outfortplayerpawn_3 (CallFunction|GetAvatarasValidFortPlayerPawn)
                        (:ValidFortPlayerPawn
                          (if (Variables|/Script/FortniteGame.FortPawn|GetbSecondaryInputHeld _outfortplayerpawn_3)
                            (Utilities|IsValid (Ability|GetFortAbilitySystemComponent _outfortplayerpawn_3)
                              (:Is%20Valid
                                (Abilities|TryActivateAbilitiesbyTag (Ability|GetFortAbilitySystemComponent _outfortplayerpawn_3) (GameplayTags|MakeLiteralGameplayTagContainer "(GameplayTags=((TagName=\"RiftGauntlets.Ability.Ascend\")))") true)))))))))
                (:Triggered
                  (CallFunction|SetEnableSlamIntentDetection false))))
            (:Triggered
              (CallFunction|SetEnableSlamIntentDetection true))))))))

(event Custom|SetEnableSlamIntentDetection
  (params (bEnable :type Boolean))
  ;[0.15 0.07146 0.1497 0.5] SetEnableSlamIntentDetection:If intent detection is true, start a ticking timer that checks to see if the player would want to slam.
  (Ability|HasAuthority
    (:False
      (Utilities|FlowControl|DeltaGateWildcard (Variables|Transient|GetbSlamIntentDetectEnable) bEnable
        (:Success
          (if bEnable
            ;; HandleSlamDetect will auto ticking itself until the intent is disabled
            (CallFunction|HandleSlamDetect)))))))

(event Custom|HandleSlamDetect
  (bind _avatarplayerpawn (Variables|Transient|GetAvatarPlayerPawn))
  ;[0.004789 0 0.15 0.5] HandleSlamDetect: Check the player's aim angle and use a ground trace to see if the player is intending to perform a dash. If so, spawn a telegraph actor on the ground at that location. If telegraph actor already exists, update its location.
  (Ability|HasAuthority
    (:False
      (if (Variables|Transient|GetbSlamIntentDetectEnable)
        (bind _returnvalue (Utilities|Time|SetTimerforNextTickbyEvent (EventDispatchers|CreateEvent :!event Custom|HandleSlamDetect)))
        (Variables|Transient|SetSlamDetectTimerHandle _returnvalue)
        (bind _avatarplayerpawn_2 (Variables|Transient|GetAvatarPlayerPawn)
          (:then
            (if (GameplayTags|HasMatchingGameplayTag _avatarplayerpawn_2 "(TagName=\"RiftGauntlets.Ability.Dash\")")
              (bind _telegraphactor (Variables|Transient|GetTelegraphActor)
                (:then
                  (Rendering|SetActorHiddenInGame _telegraphactor true))
                (:else))
              (else
                (Utilities|IsValid (Weapon|GetCurrentWeapon _avatarplayerpawn_2)
                  (:Is%20Valid
                    (if (> (FortWeapon|GetLocalMagazineAmmoCount (Weapon|GetCurrentWeapon _avatarplayerpawn_2)) 0)
                      (if (< (Math|Trig|Acos(Degrees) (Math|Vector|DotProduct (Math|Vector|GetForwardVector (Pawn|GetControlRotation _avatarplayerpawn :!class /Script/Engine.Pawn)) "0.000000,0.000000,-1.000000")) (ScalableFloat|GetCachedScalableFloatValue (Variables|Hotfix|GetHotfix_SlamIntentAngle)))
                        (bind _returnvalue_2 (Targeting|GetTargetingTransform _avatarplayerpawn "Camera" 0))
                        (bind (_outhit _returnvalue_3) (Collision|LineTraceByChannel (.location _returnvalue_2) (+ (.location _returnvalue_2) (* (Math|Vector|GetForwardVector (.rotation _returnvalue_2)) (ScalableFloat|GetCachedScalableFloatValue (Variables|Hotfix|GetHotfix_SlamIntentDistance)))) "TraceTypeQuery10" false (Utilities|Array|MakeArray _avatarplayerpawn) "None" true "(R=1.000000,G=0.000000,B=0.000000,A=1.000000)" "(R=0.000000,G=1.000000,B=0.000000,A=1.000000)" 2.0))
                        (if _returnvalue_3
                          (CallFunction|SetEnableSlamDetectGameplayTag true)
                          (if (Variables|Transient|GetbSlamIntentGameplayTagAdded)
                            (bind _telegraphactor_2 (Variables|Transient|GetTelegraphActor)
                              (:then
                                (Rendering|SetActorHiddenInGame _telegraphactor_2 false)
                                (Transformation|SetActorLocation _telegraphactor_2 (.Location (Collision|BreakHitResult _outhit)) false false))
                              (:else
                                (bind _returnvalue_4 (Game|SpawnActorfromClass "/RiftGauntletsGameplay/Gameplay/Dash/B_RiftGauntlets_DashSlamTelegraph.B_RiftGauntlets_DashSlamTelegraph_C" (Math|Transform|MakeTransform (.Location (Collision|BreakHitResult _outhit))) "AlwaysSpawn" "MultiplyWithRoot" (Variables|Transient|GetAvatarPlayerPawn) 0))
                                (bind _output_get (Variables|Transient|SetTelegraphActor _returnvalue_4))
                                (Rendering|SetOnlyVisibletoOwner _output_get true))))
                          (else
                            _))
                        (else
                          _))
                      (CallFunction|SetEnableSlamDetectGameplayTag false)
                      (bind _telegraphactor_3 (Variables|Transient|GetTelegraphActor)
                        (:then
                          (Rendering|SetActorHiddenInGame _telegraphactor_3 true))
                        (:else))
                      (else
                        _)))
                  (:Is%20Not%20Valid
                    _)))))
          (:else
            _))
        (else
          (Utilities|Time|ClearandInvalidateTimerbyHandle (Variables|Transient|GetSlamDetectTimerHandle)))))))

(event Ability|EventOnEndAbility
  (params (bWasCancelled :type Boolean))
  ;[0.15 0.07694 0.001921 0.5] OnEndAbility: Clean up.
  (Utilities|FlowControl|Sequence
    (:then_0
      (GameplayTasks|EndTask (Variables|Transient|GetAsyncTasks))
      (GameplayTasks|EndTask (Variables|Transient|GetAsyncTask_WaitCooldownEndAscend))
      (CallFunction|SetEnableSlamIntentDetection false)
      (CallFunction|SetEnableSlamDetectGameplayTag false)
      (Utilities|Time|ClearandInvalidateTimerbyHandle (Variables|Transient|GetSlamDetectTimerHandle))
      (bind _telegraphactor (Variables|Transient|GetTelegraphActor)
        (:then
          (Actor|DestroyActor _telegraphactor))
        (:else)))
    (:then_1
      (|CallParentFunction bWasCancelled))))

(fn SetEnableSlamDetectGameplayTag
  (params (bEnable :type Boolean))
  (bind _slamintentgameplaytag (Variables|Default|GetSlamIntentGameplayTag))
  ;[0.107 0.1031 0.15 0.5] SetEnableSlamDetectGameplayTag: Adds a loose gameplay on the client to track the intention. This is used later when the dash is activated. It is also used for HUD reasons.
  (Ability|HasAuthority
    (:False
      (Utilities|FlowControl|DeltaGateWildcard (Variables|Transient|GetbSlamIntentGameplayTagAdded) bEnable
        (:Success
          (bind _avatarplayerpawn (Variables|Transient|GetAvatarPlayerPawn)
            (:then
              (if bEnable
                (Ability|GameplayEffect|AddLooseGameplayTags _avatarplayerpawn _slamintentgameplaytag)
                (else
                  (Ability|GameplayEffect|RemoveLooseGameplayTags _avatarplayerpawn _slamintentgameplaytag))))
            (:else)))))))
