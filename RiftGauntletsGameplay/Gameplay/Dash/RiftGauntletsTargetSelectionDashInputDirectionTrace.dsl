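; Targeting trace selection task for the Rift Gauntlets dash that aims the trace along the player's movement input.
; Its GetTraceDirection override casts the source to a Fort player pawn and flattens the movement component's current acceleration to a normalized 2D input direction.
; It rotates that direction about its own right vector by the negated camera pitch, so the trace follows vertical look.
; A Sequence fallback returns the parent's direction if the casts fail.
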(fn GetTraceDirection
  (params (TargetingHandle :type Targeting%20Request%20Handle%20Structure%20%28by%20ref%29))
  (bind _returnvalue (|CallParentFunction TargetingHandle))
  (Utilities|FlowControl|Sequence
    (:then_0
      (bind _asfort_player_pawn (Utilities|Casting|CastToFortPlayerPawn (Utilities|Struct|BreakTargetingSourceContext (TargetingSystem|TargetingTypes|GetTargetingSourceContext (Targeting|GetTargetingSubsystem TargetingHandle) TargetingHandle))))
      (bind _ascharacter_movement_component (Utilities|Casting|CastToCharacterMovementComponent (Pawn|GetMovementComponent _asfort_player_pawn)))
      (return (Math|Vector|RotateVectorAroundAxis (Math|Vector|Normalize2D(Vector) (Pawn|Components|CharacterMovement|GetCurrentAcceleration _ascharacter_movement_component)) (* (.pitch (.rotation (Targeting|GetTargetingTransform _asfort_player_pawn "Camera" 0))) -1.0) (Math|Vector|GetRightVector (Math|Rotator|MakeRotfromX (Math|Vector|Normalize2D(Vector) (Pawn|Components|CharacterMovement|GetCurrentAcceleration _ascharacter_movement_component)))))))
    (:then_1
      (return _returnvalue))))
